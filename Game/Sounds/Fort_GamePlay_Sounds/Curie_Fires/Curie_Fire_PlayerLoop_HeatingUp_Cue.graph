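# Sound cue for the Curie fire player "heating up" sound, built from small fire-start waves.
# A Random node picks one of seven equally weighted SmallStart variations, and a Modulator varies its pitch; no wave player is set to loop, despite the Loop name.
# It uses small-fire attenuation and a received hit-notification sound class.
# A concurrency override allows one instance per owner, stopping the oldest.
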SoundCue /Game/Sounds/Fort_GamePlay_Sounds/Curie_Fires/Curie_Fire_PlayerLoop_HeatingUp_Cue
root=Modulator_1
nodes [
 {id=Modulator_1, children [Random_0], properties {pitchMax=1.2, pitchMin=0.9}}
 {
  id=Random_0
  children [WavePlayer_0, WavePlayer_1, WavePlayer_2, WavePlayer_3, WavePlayer_10, WavePlayer_11, WavePlayer_12]
  properties {weights [1,1,1,1,1,1,1]}
 }
 {id=WavePlayer_0, properties {soundWaveAssetPtr=&/Curie_Fire_SmallStart_01}}
 {id=WavePlayer_1, properties {soundWaveAssetPtr=&/Curie_Fire_SmallStart_02}}
 {id=WavePlayer_10, properties {soundWaveAssetPtr=&/Curie_Fire_SmallStart_05}}
 {id=WavePlayer_11, properties {soundWaveAssetPtr=&/Curie_Fire_SmallStart_06}}
 {id=WavePlayer_12, properties {soundWaveAssetPtr=&/Curie_Fire_SmallStart_07}}
 {id=WavePlayer_2, properties {soundWaveAssetPtr=&/Curie_Fire_SmallStart_03}}
 {id=WavePlayer_3, properties {soundWaveAssetPtr=&/Curie_Fire_SmallStart_04}}
]
properties {
 attenuationSettings=&/Curie_Fire_Small_01_Att
 bOverrideConcurrency=true
 maxDistance=2500
 soundClassObject=/Game/Sounds/FortSoundClassesAndModes/Fort_2DHitNotifications_Receive_PSM
 volumeMultiplier=1
 concurrencyOverrides {bLimitToOwner=true, maxCount=1, resolutionRule=StopOldest}
}
