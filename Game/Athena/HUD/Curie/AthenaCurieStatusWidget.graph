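# HUD status widget that shows animated fire and glow for a player's Curie Drying and attached-Fire states.
# Adding Drying plays a heat-up animation and sound, while adding Fire plays OnFire; both increment one shared request counter.
# Removing Fire reverses OnFire; Drying reverses only when the counter drains, with an error print for excess removals.
# Animation event tracks drive material-based flames and toggle widget visibility according to forward playback.
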WidgetBlueprint /Game/Athena/HUD/Curie/AthenaCurieStatusWidget
animations [
 Drying {
  binding_count=3
  length_seconds=0.5037
  object=&self:Drying
  sequence {
   bindings [
    Image_FireAdd {
     id=#1
     possessed_class=Image
     tracks [
      {
       type=MovieSceneWidgetMaterialTrack
       sections [{
         type=MovieSceneParameterSection
         channels [
          DFActivate {type=MovieSceneScriptingFloatChannel, keys [[0,0.9],[30,1]]}
          ActivateAdjust {type=MovieSceneScriptingFloatChannel, keys [[0,-1],[26,0.1545],[30,-0.0579]]}
         ]
        }]
      }
      {
       type=MovieSceneFloatTrack
       display_name=Render Opacity
       sections [{type=MovieSceneFloatSection, end_frame=30, start_frame=0, channels [None {type=MovieSceneScriptingFloatChannel, const [1,1]}]}]
      }
     ]
    }
    Image_glow {
     id=#2
     possessed_class=Image
     tracks [{
       type=MovieSceneFloatTrack
       display_name=Render Opacity
       sections [{type=MovieSceneFloatSection, channels [None {type=MovieSceneScriptingFloatChannel, keys [[0,0],[30.219,1]]}]}]
      }]
    }
    Image_Fire {
     id=#3
     possessed_class=Image
     tracks [
      {
       type=MovieSceneWidgetMaterialTrack
       sections [{
         type=MovieSceneParameterSection
         channels [
          DFActivate {type=MovieSceneScriptingFloatChannel, keys [[0,0.9],[30,1]]}
          ActivateAdjust {type=MovieSceneScriptingFloatChannel, keys [[0,-1],[26,0.1545],[30,-0.0579]]}
         ]
        }]
      }
      {
       type=MovieSceneFloatTrack
       display_name=Render Opacity
       sections [{type=MovieSceneFloatSection, end_frame=30, start_frame=0, channels [None {type=MovieSceneScriptingFloatChannel, const [1,1]}]}]
      }
     ]
    }
   ]
   display_rate {numerator=60}
   tracks [{
     type=MovieSceneEventTrack
     sections [{
       type=MovieSceneEventTriggerSection
       channels [None {type=MovieSceneScriptingEventChannel, const ["<Struct 'MovieSceneEvent' (0x4679f55e2e0) {}>", 1]}]
      }]
    }]
  }
 }
 OnFire {
  binding_count=2
  length_seconds=0.6
  object=&self:OnFire
  sequence {
   bindings [
    Image_Fire {
     id=#4
     possessed_class=Image
     tracks [
      {
       type=MovieSceneWidgetMaterialTrack
       sections [{
         type=MovieSceneParameterSection
         channels [
          DFMask {type=MovieSceneScriptingFloatChannel, keys [[0,1.0078],[12,0.8]]}
          FlameExp {type=MovieSceneScriptingFloatChannel, keys [[0,-0.9704],[8,-2.25],[12,-2.2641]]}
          TopAdjust {type=MovieSceneScriptingFloatChannel, keys [[0,0.098],[5,0.7673],[12,0.05]]}
          ActivateAdjust {type=MovieSceneScriptingFloatChannel, const [0,1]}
          DFActivate {type=MovieSceneScriptingFloatChannel, const [1,1]}
          Color1R {type=MovieSceneScriptingFloatChannel, keys [[0,0.5],[12,1]]}
          Color1G {type=MovieSceneScriptingFloatChannel, keys [[0,0.1],[12,0.2]]}
          Color1B {type=MovieSceneScriptingFloatChannel, keys [[0,0],[12,0.1]]}
          Color1A {type=MovieSceneScriptingFloatChannel, const [1,2]}
          Color2R {type=MovieSceneScriptingFloatChannel, keys [[0,0.4],[12,0.8]]}
          Color2G {type=MovieSceneScriptingFloatChannel, keys [[0,0.2],[12,0.4]]}
          Color2B {type=MovieSceneScriptingFloatChannel, const [0.1,2]}
          Color2A {type=MovieSceneScriptingFloatChannel, const [1,2]}
          Speed1R {type=MovieSceneScriptingFloatChannel, const [-0.1,2]}
          Speed1G {type=MovieSceneScriptingFloatChannel, keys [[0,0.3],[10,0.8]]}
          Speed1B {type=MovieSceneScriptingFloatChannel, const [0,1]}
          Speed1A {type=MovieSceneScriptingFloatChannel}
          Speed2R {type=MovieSceneScriptingFloatChannel, const [-0.2,2]}
          Speed2G {type=MovieSceneScriptingFloatChannel, keys [[0,0.3],[10,0.65]]}
          Speed2B {type=MovieSceneScriptingFloatChannel}
          Speed2A {type=MovieSceneScriptingFloatChannel}
         ]
        }]
      }
      {
       type=MovieSceneFloatTrack
       display_name=Render Opacity
       sections [{type=MovieSceneFloatSection, end_frame=12, start_frame=0, channels [None {type=MovieSceneScriptingFloatChannel, const [1,1]}]}]
      }
     ]
    }
    Image_FireAdd {
     id=#5
     possessed_class=Image
     tracks [
      {
       type=MovieSceneWidgetMaterialTrack
       sections [{
         type=MovieSceneParameterSection
         channels [
          DFMask {type=MovieSceneScriptingFloatChannel, keys [[0,1.0078],[12,0.2]]}
          DFScale {type=MovieSceneScriptingFloatChannel, keys [[0,1],[12,1.1]]}
          Firepower {type=MovieSceneScriptingFloatChannel, keys [[0,0.8832],[12,2.9848]]}
          FlameExp {type=MovieSceneScriptingFloatChannel, keys [[0,-0.9704],[5,-2.2641]]}
          TopAdjust {type=MovieSceneScriptingFloatChannel, keys [[0,0.098],[8,0.75],[12,0.745]]}
          ActivateAdjust {type=MovieSceneScriptingFloatChannel, const [0,1]}
          DFActivate {type=MovieSceneScriptingFloatChannel, const [1,1]}
          Speed1R {type=MovieSceneScriptingFloatChannel, keys [[0,0],[10,-0.1]]}
          Speed1G {type=MovieSceneScriptingFloatChannel, keys [[0,0.1],[10,0.4]]}
          Speed1B {type=MovieSceneScriptingFloatChannel}
          Speed1A {type=MovieSceneScriptingFloatChannel}
          Speed2R {type=MovieSceneScriptingFloatChannel, keys [[0,0],[10,-0.2]]}
          Speed2G {type=MovieSceneScriptingFloatChannel, keys [[0,0.1],[10,0.8]]}
          Speed2B {type=MovieSceneScriptingFloatChannel}
          Speed2A {type=MovieSceneScriptingFloatChannel}
         ]
        }]
      }
      {
       type=MovieSceneFloatTrack
       display_name=Render Opacity
       sections [{type=MovieSceneFloatSection, end_frame=12, start_frame=0, channels [None {type=MovieSceneScriptingFloatChannel, const [1,1]}]}]
      }
     ]
    }
   ]
   tracks [{
     type=MovieSceneEventTrack
     sections [{
       type=MovieSceneEventTriggerSection
       channels [None {
         type=MovieSceneScriptingEventChannel
         keys [[0, "<Struct 'MovieSceneEvent' (0x4679f55e1a0) {}>"], [12, "<Struct 'MovieSceneEvent' (0x4679f55e9c0) {}>"]]
        }]
      }]
    }]
  }
 }
]
graphs [
 DebugString {nodes [
   n1 = DebugString!() -> (then, GameplayTag:Gameplay Tag Structure, Add:bool=~) {is=title K2Node_FunctionEntry}
   n2 = HasState!(self:Athena Curie Status Widget Object Reference, CurieState:Gameplay Tag Structure=(TagName="Curie.State.ElementAttached.Fire")) -> (ReturnValue:bool)
   n3 = GetDebugStringFromGameplayTag!(GameplayTag:Gameplay Tag Structure=@n1) -> (ReturnValue:String)
   n4 = "Build String (Boolean)"!(AppendTo:String, Prefix:String="- Has Drying - ", InBool:bool=HasState(CurieState=(TagName="Curie.State.Drying")), Suffix:String) -> (ReturnValue:String)
   > PrintString!(
    execute
    InString=Append(
     A=Select(Option 0:String="Remove - ", Option 1:String="Add - ", Index=@n1.Add) -> (ReturnValue:String)
     B="Build String (Boolean)"(AppendTo=@n3, Prefix="- Has Fire - ", InBool=@n2, Suffix=@n4)
    )
    bPrintToScreen
    bPrintToLog
    TextColor
    Duration
    Key
   ) -> (then)
   > Return Node()
  ]}
 EventGraph {nodes [
   n1 = Event OnCurieStateAdded!() -> (OutputDelegate, then, AddedState:"Gameplay Tag Structure (by ref)") {is=title K2Node_Event}
   > n2 = Branch(Condition=MatchesTag!(TagOne:Gameplay Tag Structure=@n1.AddedState, TagTwo:Gameplay Tag Structure=(TagName="Curie.State.Drying"), bExactMatch:bool=false) -> (ReturnValue:bool))
   > PlayAnimationForward(InAnimation=Get Drying() -> (Drying:Widget Animation Object Reference))
   > PlaySound2D(Sound=/Game/Sounds/Fort_GamePlay_Sounds/Curie_Fires/Curie_Fire_PlayerLoop_HeatingUp_Cue)
   > Increment Int(Value=Get DryingAnimRequestCount() -> (DryingAnimRequestCount:int=0))
   Branch(execute=@n2.else, Condition=MatchesTag(TagOne=@n1.AddedState, TagTwo=(TagName="Curie.State.ElementAttached.Fire"), bExactMatch=false))
   > PlayAnimationForward(InAnimation=Get OnFire() -> (OnFire:Widget Animation Object Reference))
   > Increment Int(Value=Get DryingAnimRequestCount() -> (DryingAnimRequestCount=0))
   n3 = Event OnCurieStateRemoved!() -> (OutputDelegate, then, RemovedState:"Gameplay Tag Structure (by ref)") {is=title K2Node_Event}
   > UpdateRemovedState!(execute, self, RemovedState:"Gameplay Tag Structure (by ref)"=@n3.RemovedState) -> (then)
   SetPlayerState!(execute, self:Athena Curie Status Widget Object Reference, PlayerState:Fort Player State Athena Object Reference) -> (then)
   HasState()
  ]}
 Sequencer Events {nodes [
   n1 = MarkFireActive() {is=title K2Node_CustomEvent}
   n2 = IsAnimationPlayingForward!(execute=[@n1.then, @n4.then, @n5.then], self:User Widget Object Reference, InAnimation:Widget Animation Object Reference=Get Drying()) -> (then, ReturnValue:bool)
   > n3 = IsAnimationPlayingForward(InAnimation=Get OnFire())
   > SetVisibility(InVisibility=Select(Option 0:ESlateVisibility Enum=Collapsed, Option 1:ESlateVisibility Enum=HitTestInvisible, Index=OR Boolean(A=@n3, B=@n2)) -> (ReturnValue:ESlateVisibility Enum))
   n4 = OnFireEvent() {is=title K2Node_CustomEvent}
   n5 = DryingFireEventActive() {is=title K2Node_CustomEvent}
  ]}
 UpdateRemovedState {nodes [
   UpdateRemovedState!() -> (then, RemovedState:"Gameplay Tag Structure (by ref)") {is=title K2Node_FunctionEntry}
   > Set IsDryingActive(IsDryingActive:bool="integer > integer"(A=Get DryingAnimRequestCount() -> (DryingAnimRequestCount=0), B=0) {is=title K2Node_CallFunction}) -> (Output_Get:bool)
   > n1 = Sequence()
   > n2 = Branch(Condition=MatchesTag(TagOne=Get RemovedState() -> (RemovedState:"Gameplay Tag Structure (by ref)"), TagTwo=(TagName="Curie.State.Drying"), bExactMatch=false))
   > Decrement Int(Value=Get DryingAnimRequestCount() -> (DryingAnimRequestCount=0))
   Branch(execute=@n2.else, Condition=MatchesTag(TagOne=Get RemovedState(), TagTwo=(TagName="Curie.State.ElementAttached.Fire"), bExactMatch=false))
   > PlayAnimationReverse(InAnimation=Get OnFire())
   > Decrement Int(Value=Get DryingAnimRequestCount() -> (DryingAnimRequestCount=0))
   n3 = Get DryingAnimRequestCount() -> (DryingAnimRequestCount=0)
   Branch(
    execute=@n1.then_1
    Condition=AND Boolean(
     A=Get IsDryingActive() -> (IsDryingActive:bool)
     B="integer <= integer"!(A:int=@n3, B:int=0) -> (ReturnValue:bool) {is=title K2Node_CallFunction}
    )
   )
   > PlayAnimationReverse(InAnimation=Get Drying())
   > n4 = Branch(Condition="Equal (Integer)"(A=Get DryingAnimRequestCount() -> (DryingAnimRequestCount=0), B=0) {is=title K2Node_CallFunction})
   PrintString(
    execute=@n4.else
    InString="ERROR:  AthenaCurieStatusWidget received more state removals than adds"
    TextColor=(R=0.4271,G=0.029,B=0.029,A=1)
   )
   > Set DryingAnimRequestCount(DryingAnimRequestCount:int=0) -> (Output_Get:int=0)
  ]}
]
variables [DryingAnimRequestCount {type=integer}]
widgets {
 info {inheritedWidgetCount=1, parentClass=/Script/FortniteUI.AthenaCurieStatusWidget, rootWidgetClass=&Overlay, widgetCount=8}
 widgets [
  Overlay_0 {widgetClassPath=&Overlay, widget=Overlay_0}
  Image_Fire {
   widgetClassPath=&Image
   bIsVariable=true
   parent=Overlay_0
   widget=Image_Fire
   properties {brush {
     bIsDynamicallyLoaded=false
     drawAs=Image
     imageType=NoImage
     mirroring=NoMirror
     resourceName=None
     resourceObject=/Game/Athena/HUD/LTM/Mash/Materials/M_UI_CurieFire_Small
     tiling=NoTile
     imageSize=<32 32>
     margin=<ltrb 0.25 0.2 0.25 0.2>
     outlineSettings {
      bUseBrushTransparency=false
      roundingType=HalfHeightRadius
      width=0
      color {colorUseRule=UseColor_Specified, specifiedColor=<rgba 0 0 0 0>}
      cornerRadii=<0 0 0 0>
     }
     tintColor {colorUseRule=UseColor_Specified, specifiedColor=<rgba 1 1 1 1>}
     uVRegion {bIsValid=false, max=<0 0>, min=<0 0>}
    }}
   slot {class=OverlaySlot, properties {horizontalAlignment=HAlign_Fill, verticalAlignment=VAlign_Fill, padding=<ltrb 0 -10 0 0>}}
  }
  Image_FireAdd {
   widgetClassPath=&Image
   bIsVariable=true
   parent=Overlay_0
   widget=Image_FireAdd
   properties {
    brush {
     bIsDynamicallyLoaded=false
     drawAs=Image
     imageType=NoImage
     mirroring=NoMirror
     resourceName=None
     resourceObject=/Game/Athena/HUD/LTM/Mash/Materials/M_UI_CurieFire_Small_Add
     tiling=NoTile
     imageSize=<32 32>
     margin=<ltrb 0 0 0 0>
     outlineSettings {
      bUseBrushTransparency=false
      roundingType=HalfHeightRadius
      width=0
      color {colorUseRule=UseColor_Specified, specifiedColor=<rgba 0 0 0 0>}
      cornerRadii=<0 0 0 0>
     }
     tintColor {colorUseRule=UseColor_Specified, specifiedColor=<rgba 1 1 1 1>}
     uVRegion {bIsValid=false, max=<0 0>, min=<0 0>}
    }
    renderTransform {angle=0, scale=<1 1.05>, shear=<0 0>, translation=<0 0>}
    renderTransformPivot=<0.5 0.9>
   }
   slot {class=OverlaySlot, properties {horizontalAlignment=HAlign_Fill, verticalAlignment=VAlign_Fill, padding=<ltrb 0 -10 0 0>}}
  }
  Image_glow {
   widgetClassPath=&Image
   bIsVariable=true
   parent=Overlay_0
   widget=Image_glow
   properties {
    brush {
     bIsDynamicallyLoaded=false
     drawAs=Image
     imageType=NoImage
     mirroring=NoMirror
     resourceName=None
     resourceObject=/Game/Athena/HUD/LTM/Wax/Materials/M_UI_TextGlow_Pulse_Add
     tiling=NoTile
     imageSize=<32 32>
     margin=<ltrb 0 0 0 0>
     outlineSettings {
      bUseBrushTransparency=false
      roundingType=HalfHeightRadius
      width=0
      color {colorUseRule=UseColor_Specified, specifiedColor=<rgba 0 0 0 0>}
      cornerRadii=<0 0 0 0>
     }
     tintColor {colorUseRule=UseColor_Specified, specifiedColor=<rgba 1 0.0924 0 1>}
     uVRegion {bIsValid=false, max=<0 0>, min=<0 0>}
    }
    renderTransform {angle=0, scale=<1.15 1>, shear=<0 0>, translation=<0 0>}
   }
   slot {class=OverlaySlot, properties {horizontalAlignment=HAlign_Fill, verticalAlignment=VAlign_Fill, padding=<ltrb -10 0 -10 -40>}}
  }
  Switcher_FlameSize {
   widgetClassPath=&CommonWidgetSwitcherLegacy
   bIsVariable=true
   parent=Overlay_0
   widget=Switcher_FlameSize
   properties {activeWidgetIndex=1, visibility=Collapsed}
   slot {class=OverlaySlot, properties {horizontalAlignment=HAlign_Fill, verticalAlignment=VAlign_Fill}}
  }
  Image_31 {
   widgetClassPath=&Image
   bIsVariable=true
   parent=Switcher_FlameSize
   widget=Image_31
   properties {brush {
     bIsDynamicallyLoaded=false
     drawAs=Box
     imageType=NoImage
     mirroring=NoMirror
     resourceName=None
     resourceObject=&M_UI_CurieFire_Small
     tiling=NoTile
     imageSize=<32 32>
     margin=<ltrb 0.25 0.2 0.25 0.2>
     outlineSettings {
      bUseBrushTransparency=false
      roundingType=HalfHeightRadius
      width=0
      color {colorUseRule=UseColor_Specified, specifiedColor=<rgba 0 0 0 0>}
      cornerRadii=<0 0 0 0>
     }
     tintColor {colorUseRule=UseColor_Specified, specifiedColor=<rgba 1 1 1 1>}
     uVRegion {bIsValid=false, max=<0 0>, min=<0 0>}
    }}
   slot {class=WidgetSwitcherSlot, properties {padding=<ltrb 0 -10 0 0>}}
  }
  Image_20 {
   widgetClassPath=&Image
   bIsVariable=true
   parent=Switcher_FlameSize
   widget=Image_20
   properties {brush {
     bIsDynamicallyLoaded=false
     drawAs=Box
     imageType=NoImage
     mirroring=NoMirror
     resourceName=None
     resourceObject=/Game/Athena/HUD/LTM/Mash/Materials/M_UI_CurieFire_Large
     tiling=NoTile
     imageSize=<32 32>
     margin=<ltrb 0.2 0.2 0.2 0.2>
     outlineSettings {
      bUseBrushTransparency=false
      roundingType=HalfHeightRadius
      width=0
      color {colorUseRule=UseColor_Specified, specifiedColor=<rgba 0 0 0 0>}
      cornerRadii=<0 0 0 0>
     }
     tintColor {colorUseRule=UseColor_Specified, specifiedColor=<rgba 1 1 1 1>}
     uVRegion {bIsValid=false, max=<0 0>, min=<0 0>}
    }}
   slot {class=WidgetSwitcherSlot}
  }
  TutorialHighlight {widgetClassPath=&TutorialHighlight}
 ]
}
